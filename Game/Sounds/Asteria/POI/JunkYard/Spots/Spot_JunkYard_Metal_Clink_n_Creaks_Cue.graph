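# Ambient spot Sound Cue that plays metal clinks and creaks for the JunkYard point of interest.
# A Random node picks among four clink and three creak waves with equal weight, feeding a Modulator that varies pitch.
# It uses a distant metal-movement attenuation asset and the outdoor ambient sound class, with bus and submix sends disabled.
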SoundCue /Game/Sounds/Asteria/POI/JunkYard/Spots/Spot_JunkYard_Metal_Clink_n_Creaks_Cue
root=Modulator_0
nodes [
 {id=Modulator_0, children [Random_1], properties {pitchMin=0.9}}
 {
  id=Random_1
  children [WavePlayer_0,WavePlayer_1,WavePlayer_2,WavePlayer_3,WavePlayer_4,WavePlayer_5,WavePlayer_6]
  properties {weights [1,1,1,1,1,1,1]}
 }
 {id=WavePlayer_0, properties {soundWaveAssetPtr=&/Spots_Metal_Clinks_01}}
 {id=WavePlayer_1, properties {soundWaveAssetPtr=&/Spots_Metal_Clinks_02}}
 {id=WavePlayer_2, properties {soundWaveAssetPtr=&/Spots_Metal_Clinks_03}}
 {id=WavePlayer_3, properties {soundWaveAssetPtr=&/Spots_Metal_Clinks_04}}
 {id=WavePlayer_4, properties {soundWaveAssetPtr=&/Spots_Metal_Creak_01}}
 {id=WavePlayer_5, properties {soundWaveAssetPtr=&/Spots_Metal_Creak_02}}
 {id=WavePlayer_6, properties {soundWaveAssetPtr=&/Spots_Metal_Creak_03}}
]
properties {
 attenuationSettings=&/Spot_Metal_Movements_Distant_Att
 bEnableBusSends=false
 bEnableSubmixSends=false
 maxDistance=7500
 soundClassObject=&Fort_Ambient_Outdoors_Athena
 volumeMultiplier=0.752
 attenuationOverrides {falloffDistance=6000, stereoSpread=10000, attenuationShapeExtents {x=2000}}
}
