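# Sound cue for ambient metal groan spot sounds at the JunkYard point of interest in Asteria.
# A Random node picks with equal weight from eleven medium and small groan waves, then a Modulator varies pitch, with a slightly lowered minimum.
# It plays in the outdoor ambient sound class with distant metal-movement attenuation, and bus and submix sends are disabled.
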SoundCue /Game/Sounds/Asteria/POI/JunkYard/Spots/Spot_JunkYard_Metal_Groans_Cue
root=Modulator_0
nodes [
 {id=Modulator_0, children [Random_1], properties {pitchMin=0.9}}
 {
  id=Random_1
  children [
   WavePlayer_0
   WavePlayer_1
   WavePlayer_2
   WavePlayer_3
   WavePlayer_4
   WavePlayer_5
   WavePlayer_6
   WavePlayer_7
   WavePlayer_8
   WavePlayer_9
   WavePlayer_14
  ]
  properties {weights [1,1,1,1,1,1,1,1,1,1,1]}
 }
 {id=WavePlayer_0, properties {soundWaveAssetPtr=&/Spots_Metal_Groan_Medium_01}}
 {id=WavePlayer_1, properties {soundWaveAssetPtr=&/Spots_Metal_Groan_Medium_02}}
 {id=WavePlayer_14, properties {soundWaveAssetPtr=&/Spots_Metal_Groan_Small_08}}
 {id=WavePlayer_2, properties {soundWaveAssetPtr=&/Spots_Metal_Groan_Medium_03}}
 {id=WavePlayer_3, properties {soundWaveAssetPtr=&/Spots_Metal_Groan_Small_01}}
 {id=WavePlayer_4, properties {soundWaveAssetPtr=&/Spots_Metal_Groan_Small_02}}
 {id=WavePlayer_5, properties {soundWaveAssetPtr=&/Spots_Metal_Groan_Small_03}}
 {id=WavePlayer_6, properties {soundWaveAssetPtr=&/Spots_Metal_Groan_Small_04}}
 {id=WavePlayer_7, properties {soundWaveAssetPtr=&/Spots_Metal_Groan_Small_05}}
 {id=WavePlayer_8, properties {soundWaveAssetPtr=&/Spots_Metal_Groan_Small_06}}
 {id=WavePlayer_9, properties {soundWaveAssetPtr=&/Spots_Metal_Groan_Small_07}}
]
properties {
 attenuationSettings=&/Spot_Metal_Movements_Distant_Att
 bEnableBusSends=false
 bEnableSubmixSends=false
 maxDistance=7500
 soundClassObject=&Fort_Ambient_Outdoors_Athena
 volumeMultiplier=0.752
 attenuationOverrides {falloffDistance=6000, stereoSpread=10000, attenuationShapeExtents {x=2000}}
}
