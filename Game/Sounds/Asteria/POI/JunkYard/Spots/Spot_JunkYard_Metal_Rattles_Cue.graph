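# Sound cue for metal rattle spot sounds at the JunkYard POI, classed as outdoor ambience.
# A Random node picks among four equally weighted junk rattle waves, then a Modulator adds pitch variation.
# A distant metal-movement attenuation asset handles spatial falloff, and bus and submix sends are disabled.
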SoundCue /Game/Sounds/Asteria/POI/JunkYard/Spots/Spot_JunkYard_Metal_Rattles_Cue
root=Modulator_0
nodes [
 {id=Modulator_0, children [Random_1], properties {pitchMin=0.9}}
 {id=Random_1, children [WavePlayer_0,WavePlayer_1,WavePlayer_2,WavePlayer_3], properties {weights [1,1,1,1]}}
 {id=WavePlayer_0, properties {soundWaveAssetPtr=&/Spots_Metal_Junk_Rattle_01}}
 {id=WavePlayer_1, properties {soundWaveAssetPtr=&/Spots_Metal_Junk_Rattle_02}}
 {id=WavePlayer_2, properties {soundWaveAssetPtr=&/Spots_Metal_Junk_Rattle_03}}
 {id=WavePlayer_3, properties {soundWaveAssetPtr=&/Spots_Metal_Junk_Rattle_04}}
]
properties {
 attenuationSettings=&/Spot_Metal_Movements_Distant_Att
 bEnableBusSends=false
 bEnableSubmixSends=false
 maxDistance=7500
 soundClassObject=&Fort_Ambient_Outdoors_Athena
 volumeMultiplier=0.752
 attenuationOverrides {falloffDistance=6000, stereoSpread=10000, attenuationShapeExtents {x=2000}}
}
