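# Sound cue for a metal sheet wobble ambient spot sound in the Asteria junkyard POI, assigned to an outdoor ambient sound class.
# A random node picks one of four equally weighted wobble recordings, and a modulator at the root varies it with a lowered minimum pitch.
# It plays at reduced volume with a distant metal-movement attenuation asset for spatial falloff, with bus and submix sends disabled.
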SoundCue /Game/Sounds/Asteria/POI/JunkYard/Spots/Spot_JunkYard_Metal_Sheet_Wobble_Cue
root=Modulator_0
nodes [
 {id=Modulator_0, children [Random_0], properties {pitchMin=0.9}}
 {id=Random_0, children [WavePlayer_10, WavePlayer_11, WavePlayer_12, WavePlayer_13], properties {weights [1,1,1,1]}}
 {id=WavePlayer_10, properties {soundWaveAssetPtr=&/Spot_Metal_Sheet_Wobble_01}}
 {id=WavePlayer_11, properties {soundWaveAssetPtr=&/Spot_Metal_Sheet_Wobble_02}}
 {id=WavePlayer_12, properties {soundWaveAssetPtr=&/Spot_Metal_Sheet_Wobble_03}}
 {id=WavePlayer_13, properties {soundWaveAssetPtr=&/Spot_Metal_Sheet_Wobble_04}}
]
properties {
 attenuationSettings=&/Spot_Metal_Movements_Distant_Att
 bEnableBusSends=false
 bEnableSubmixSends=false
 maxDistance=7500
 soundClassObject=&Fort_Ambient_Outdoors_Athena
 volumeMultiplier=0.501
 attenuationOverrides {falloffDistance=6000, stereoSpread=10000, attenuationShapeExtents {x=2000}}
}
